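; Parent blueprint for the CrewCutStanding adjustable light prop, built in the construction script from base, rod and fixture meshes.
; Mesh bounds and FixtureLength place the fixture and fill the gap with a stretched rod or rotated instanced segments.
; The first light component snaps to the fixture's socket, or lights are optionally removed.
; It also adds optional capsule collision, applies materials and tints, and exposes the rod mesh to UEFN.
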(event EventBeginPlay)

(event Collision|EventActorBeginOverlap
  (params (OtherActor :type Actor%20Object%20Reference)))

(event EventTick
  (params (DeltaSeconds :type Float%20%28single-precision%29)))

(fn AddCapsuleCollision
  (bind _returnvalue (AddComponent|NODE_AddCapsuleComponent-3 :RelativeTransform (Math|Transform|MakeTransform (Math|Vector|MakeVector 0.0 0.0 (- (/ (Variables|AdjustableLight|GetFixtureLength) 2.0) (- (Variables|AdjustableLight|GetFixtureLength) (Variables|AdjustableLight|GetBaseZBounds)))))))
  (bind _output_get (Variables|Default|SetCapsuleCollision _returnvalue))
  (Components|Capsule|SetCapsuleHalfHeight _output_get (/ (Variables|AdjustableLight|GetFixtureLength) 2.0) true)
  (for _array_element (Utilities|Array|MakeArray (Variables|AdjustableLight|GetLightFixture) (Variables|AdjustableLight|GetDownRod) (Variables|Debug|GetStaticMeshComponent))
    (Collision|SetCollisionEnabled _array_element "NoCollision")))

(fn BuildExposedAssetProviderVersion(Implementation)
  (params (Version :type Fort%20Exposed%20Asset%20Provider%20Version%20Structure%20%28by%20ref%29))
  (|CallParentFunction Version)
  (ExposedAssetProvider|AddVersion Version 1)
  (return))

(fn CollectAssetReferencestoAllow(Implementation)
  (params (Collector :type Fort%20Exposed%20Asset%20Reference%20Collector%20Structure%20%28by%20ref%29))
  (|CallParentFunction Collector)
  (bind _returnvalue (Fortnite|Utilities|ResolveComponentFromActor self (Variables|AdjustableLight|GetDownRod)))
  (ExposedAssetProvider|CollectIndirectAssetReference Collector (Utilities|ResolveSoftReference (Variables|/Script/Engine.StaticMeshComponent|GetStaticMesh _returnvalue))) ; Collect the mesh to allow it to be referenced in UEFN
  (return))

(fn EnableCollisionOnMeshParts
  (for _array_element (Utilities|Array|MakeArray (Variables|AdjustableLight|GetDownRod) (Variables|AdjustableLight|GetLightFixture))
    (Collision|SetCollisionProfileName _array_element "Default" true :!class /Script/Engine.PrimitiveComponent)
    (Development|PrintString (Utilities|String|ToString(Name) (Collision|GetCollisionProfileName (stale _array_element_2)))))
  (for _array_element_2 (Utilities|Array|MakeArray (Variables|AdjustableLight|GetDownRod) (Variables|AdjustableLight|GetLightFixture))
    _))

(fn GetCurieCollisionComponentOverride
  (return (Variables|Default|GetCapsuleCollision)))

(fn GetMaterialFromRootMesh
  (if (Variables|AdjustableLight|GetUseMaterialFromRootMesh)
    (Rendering|Material|SetMaterial (Variables|AdjustableLight|GetLightFixture) 0 (Material|GetMID (Variables|Debug|GetStaticMeshComponent)))))

(fn GetMeshBounds
  (bind (_min _max) (Components|StaticMesh|GetLocalBounds (Variables|AdjustableLight|GetDownRod)))
  (Variables|AdjustableLight|SetBaseZBounds (.z (Components|StaticMesh|GetLocalBounds (Variables|Debug|GetStaticMeshComponent))))
  (Variables|AdjustableLight|SetFixtureZBounds (.z (Components|StaticMesh|GetLocalBounds (Variables|AdjustableLight|GetLightFixture))))
  (Variables|AdjustableLight|SetRodZBounds (.z _max)))

(fn GetTintColorFromTextureData
  (bind _outtexturedataarray (Building|GetBuildingTextureData))
  (for _array_element (Utilities|Array|MakeArray (Variables|Debug|GetStaticMeshComponent) (Variables|AdjustableLight|GetDownRod) (Variables|AdjustableLight|GetLightFixture))
    (Default|SetColoronSecondaryMesh (Math|Conversions|ToLinearColor(Color) (Variables|/Script/FortniteGame.BuildingTextureData|GetTintColor (Utilities|Array|Get(acopy) _outtexturedataarray))) _array_element)))

(fn NewFunction
  (params (Condition :type Boolean))
  (if Condition
    (Components|DestroyComponent (Variables|Default|GetTODLightingComponent))
    (Components|DestroyComponent (Variables|Default|GetPointLight))))

(fn SetDownRodMaterial
  (bind _downrod (Variables|AdjustableLight|GetDownRod))
  (bind _returnvalue (Rendering|Material|GetMaterial _downrod 0))
  (bind _downrod_2 (Variables|AdjustableLight|GetDownRod))
  (bind _returnvalue_2 (Rendering|Material|GetMaterial _downrod_2 0))
  (if (Variables|AdjustableLight|GetUseGeoForDownrod)
    (if (Utilities|IsValid _returnvalue_2 :!class /Script/Engine.KismetSystemLibrary)
      (Rendering|Material|CreateDynamicMaterialInstance (Variables|AdjustableLight|GetISM_DownRod) 0 _returnvalue_2 "None" :!class /Script/Engine.PrimitiveComponent))
    (elif (Utilities|IsValid _returnvalue :!class /Script/Engine.KismetSystemLibrary)
      (bind _returnvalue_3 (Rendering|Material|CreateDynamicMaterialInstance (Variables|AdjustableLight|GetDownRod) 0 _returnvalue "None" :!class /Script/Engine.PrimitiveComponent))
      (Rendering|Material|SetScalarParameterValue _returnvalue_3 "UV Scale Vertical" (Variables|AdjustableLight|GetRodScale) :!class /Script/Engine.MaterialInstanceDynamic))))

(fn SetFixturePosition
  (Transformation|SetRelativeLocation (Variables|AdjustableLight|GetLightFixture) (Math|Vector|MakeVector 0.0 0.0 (+ (Variables|AdjustableLight|GetFixtureLength) (Variables|AdjustableLight|GetFixtureZBounds))) false false))

(fn SetLightPosition
  (bind _returnvalue (Actor|GetComponentsByClass "/Script/Engine.LightComponent"))
  (bind _lightfixture (Variables|AdjustableLight|GetLightFixture))
  (bind _returnvalue_2 (Transformation|GetAllSocketNames _lightfixture))
  (if (> (Utilities|Array|Length _returnvalue) 0)
    (if (> (Utilities|Array|Length _returnvalue_2) 0)
      (Transformation|SetWorldTransform (Utilities|Array|Get(acopy) _returnvalue) (Math|Transform|MakeTransform (+ (Variables|AdjustableLight|GetLightOffset) (.location (Transformation|GetSocketTransform _lightfixture (Utilities|Array|Get(acopy) _returnvalue_2) "RTS_World"))) (.rotation (Transformation|GetWorldTransform (Utilities|Array|Get(acopy) _returnvalue))) (.scale (Transformation|GetSocketTransform _lightfixture (Utilities|Array|Get(acopy) _returnvalue_2) "RTS_World"))) false false))))

(fn SetNumberofRodMeshes
  (bind _downrod (Variables|AdjustableLight|GetDownRod))
  (bind _staticmesh (Variables|/Script/Engine.StaticMeshComponent|GetStaticMesh _downrod))
  (bind _returnvalue (AddComponent|NODE_AddInstancedStaticMeshComponent-0))
  (bind _output_get (Variables|AdjustableLight|SetISM_DownRod _returnvalue))
  (if (Utilities|IsValid _staticmesh :!class /Script/Engine.KismetSystemLibrary)
    (Components|StaticMesh|SetStaticMesh _output_get _staticmesh)
    (bind _output_get_2 (LocalVariables|SetNumberOfInstances (/ (Utilities|Operators|Subtract (Variables|AdjustableLight|GetFixtureLength) (* (Variables|AdjustableLight|GetBaseZBounds) -1.0) (* (Variables|AdjustableLight|GetFixtureZBounds) -1.0)) (* (Variables|AdjustableLight|GetRodZBounds) -1.0))))
    (for _index (range (+ (Math|Float|Floor _output_get_2) 1))
      (Components|InstancedStaticMesh|AddInstance _output_get (Math|Transform|MakeTransform (Math|Vector|MakeVector 0.0 0.0 (+ (* _index (Variables|AdjustableLight|GetRodZBounds)) (Variables|AdjustableLight|GetBaseZBounds))) (Math|Rotator|MakeRotator 0.0 0.0 (Math|Conversions|ToFloat(Integer) (* _index (Variables|AdjustableLight|GetChainlinkRotationAmount))))) false)
      (Rendering|SetVisibility (Variables|AdjustableLight|GetDownRod) false false))
    (else
      (Development|PrintString "DownRod mesh not found"))))

(fn SetRodPosition
  (bind _output_get (Variables|AdjustableLight|SetRodScale (/ (Utilities|Operators|Subtract (Variables|AdjustableLight|GetFixtureLength) (* (Variables|AdjustableLight|GetBaseZBounds) -1.0) (* (Variables|AdjustableLight|GetFixtureZBounds) -1.0)) (Variables|AdjustableLight|GetRodZBounds))))
  (Transformation|SetRelativeScale3D (Variables|AdjustableLight|GetDownRod) (Math|Vector|MakeVector 1.0 1.0 _output_get))
  (Transformation|SetRelativeLocation (Variables|AdjustableLight|GetDownRod) (Math|Vector|MakeVector 0.0 0.0 (Variables|AdjustableLight|GetBaseZBounds)) false false))

(fn SetTintColorOnMeshParts
  (for _array_element (Utilities|Array|MakeArray (Variables|AdjustableLight|GetDownRod) (Variables|AdjustableLight|GetLightFixture))
    (Default|SetColoronSecondaryMesh (Variables|AdjustableLight|GetEmissiveTint) _array_element)))

(fn ConstructionScript
  (|CallParentFunction)
  (CallFunction|NewFunction (Variables|Default|GetRemoveLightComponents))
  (CallFunction|GetMeshBounds)
  (if (!= (Variables|AdjustableLight|GetUseMeshCollision) true)
    (CallFunction|AddCapsuleCollision))
  (CallFunction|SetFixturePosition)
  (CallFunction|SetLightPosition)
  (if (Variables|AdjustableLight|GetUseGeoForDownrod)
    (Default|SetNumberofRodMeshes)
    (else
      (CallFunction|SetRodPosition)))
  (CallFunction|ValidateBuildingTextureData)
  (if (Variables|AdjustableLight|GetHasTextureData)
    (CallFunction|GetTintColorfromTextureData)
    (else
      (CallFunction|GetMaterialfromRootMesh)))
  (CallFunction|SetDownRodMaterial)
  (CallFunction|SetTintColoronMeshParts))

(fn ValidateBuildingTextureData
  (bind _outtexturedataarray (Building|GetBuildingTextureData))
  (Variables|AdjustableLight|SetHasTextureData (> (Utilities|Array|Length _outtexturedataarray) 0))
  (return))
